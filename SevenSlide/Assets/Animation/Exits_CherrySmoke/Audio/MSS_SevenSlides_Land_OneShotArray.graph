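# MetaSound Source for the SevenSlide landing sound, playing one of several roll-tumble wave variations as a mono one-shot.
# On play, an Array Shuffle node picks the next wave from the Sounds input and triggers a mono Wave Player.
# The player feeds the mono output and signals finished when the wave ends.
# Playback uses the weapon foley sound class and concurrency set at reduced volume.
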MetaSoundSource /SevenSlide/Assets/Animation/Exits_CherrySmoke/Audio/MSS_SevenSlides_Land_OneShotArray
inputs [
 Sounds {type=WaveAsset:Array}
 UE.Source.OnPlay {type=Trigger}
]
interfaces [UE.OutputFormat.Mono:1.0, UE.Source.OneShot:1.0, UE.Source:1.1]
outputs [
 "UE.OutputFormat.Mono.Audio:0" {type=Audio}
 UE.Source.OneShot.OnFinished {type=Trigger}
]
pages [Default {
  input_defaults {
   UE.Source.OnPlay=null
   Sounds [
    /SevenSlide/Assets/Animation/Exits_CherrySmoke/Audio/Waves/SevenSlides_Roll_Tumble_01
    /SevenSlide/Assets/Animation/Exits_CherrySmoke/Audio/Waves/SevenSlides_Roll_Tumble_02
    /SevenSlide/Assets/Animation/Exits_CherrySmoke/Audio/Waves/SevenSlides_Roll_Tumble_03
    /SevenSlide/Assets/Animation/Exits_CherrySmoke/Audio/Waves/SevenSlides_Roll_Tumble_04
   ]
  }
  nodes [
   n1 = Array.Shuffle:WaveAsset:Array(In Array=^Sounds, Next=^UE.Source.OnPlay) {also {defaults {Enable Shared State="<MetaSoundBuilderResult.SUCCEEDED: 0>"}}}
   WavePlayer_0 = UE.Wave Player:Mono(Play=@n1.On Next, Wave Asset=@n1.Value) {also {version [1,1]}}
  ]
  links [{dst="@UE.Source.OneShot.OnFinished", src=WavePlayer_0.On Finished}, {dst="@UE.OutputFormat.Mono.Audio:0", src=WavePlayer_0.Out Mono}]
 }]
properties {
 assetImportData=&self:AssetImportData
 sampleRate=48000
 soundClassObject=&Fort_WeaponFoley
 volume=0.3565
 concurrencySet [/SevenSlide/Assets/Animation/Exits_CherrySmoke/FortMetasound/Weapons/CONC_Fort_WeaponFoley]
}
